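; Decal actor for the JellyFish Flopper consumable that fades in when it begins play.
; The construction script creates a dynamic instance of a ChillBronco hit decal material (no caustic motion variant), stores it as Decal_MID and applies it to the Decal component.
; BeginPlay plays the FadeIn timeline from the start, writing its track value to the material's FadeIn scalar parameter on each update.
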(event EventBeginPlay
  (bind (_direction _newtrack_0) (Timeline|FadeIn PlayFromStart)
    (:Update
      (Rendering|Material|SetScalarParameterValue (Variables|Default|GetDecal_MID) "FadeIn" _newtrack_0 :!class /Script/Engine.MaterialInstanceDynamic))))

(fn ConstructionScript
  (bind _returnvalue (Rendering|Material|CreateDynamicMaterialInstance "/Game/Athena/Items/Consumables/ChillBronco/FX/MI_Decal_ChillBronco_Hit_NoCuasticMotion.MI_Decal_ChillBronco_Hit_NoCuasticMotion" "None" "None" :!class /Script/Engine.KismetMaterialLibrary))
  (bind _output_get (Variables|Default|SetDecal_MID _returnvalue))
  (Rendering|Components|Decal|SetDecalMaterial (Variables|Decal|GetDecal) _output_get))
